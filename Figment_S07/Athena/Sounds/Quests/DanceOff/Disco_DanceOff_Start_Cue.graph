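# Sound cue for the start sting of the Disco Dance Off quest in Figment Season 7.
# It plays a single Disco_DanceOff_Start wave through one wave player, with a slight volume boost.
# DanceOff_FullArea_Att sets its spatial falloff, and the Fort_DanceMusic_Quest_PSM sound class routes it as quest dance music.
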SoundCue /Figment_S07/Athena/Sounds/Quests/DanceOff/Disco_DanceOff_Start_Cue
root=WavePlayer_1
nodes [{id=WavePlayer_1, properties {soundWaveAssetPtr=&/Disco_DanceOff_Start}}]
properties {
 attenuationSettings=&/DanceOff_FullArea_Att
 maxDistance=2200
 soundClassObject=/Game/Sounds/FortSoundClassesAndModes/Fort_DanceMusic_Quest_PSM
 volumeMultiplier=1.2
}
