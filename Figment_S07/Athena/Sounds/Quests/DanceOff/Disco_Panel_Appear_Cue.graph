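# Sound cue for the disco panel appearing in the DanceOff quest content of Figment Season 7.
# A single wave player feeds the output directly, with no randomization or modulation.
# It routes through the gadgets sound class and, despite referencing the quest interactables hint attenuation asset, overrides attenuation locally with its own shape, falloff and stereo spread.
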SoundCue /Figment_S07/Athena/Sounds/Quests/DanceOff/Disco_Panel_Appear_Cue
root=WavePlayer_0
nodes [{id=WavePlayer_0, properties {soundWaveAssetPtr=&/Disco_Panel_Appear_01}}]
properties {
 attenuationSettings=/Game/Sounds/Attenuation/Fort_QuestInteractables_Hint
 bOverrideAttenuation=true
 maxDistance=2800
 soundClassObject=&Fort_Gadgets
 attenuationOverrides {falloffDistance=2000, stereoSpread=500, attenuationShapeExtents {x=800}}
}
