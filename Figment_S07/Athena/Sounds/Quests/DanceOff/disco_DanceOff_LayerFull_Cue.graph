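# Sound cue that plays the full music layer for the DanceOff disco quest, routed to a quest dance music sound class.
# Looping music layers 2 and 3 are mixed under a full-area attenuation named as non-spatial.
# A looping reverb version of layer 1 runs through a separate far-area attenuation.
# A root mixer combines both branches, so each part follows its own distance settings.
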SoundCue /Figment_S07/Athena/Sounds/Quests/DanceOff/disco_DanceOff_LayerFull_Cue
root=Mixer_2
nodes [
 {id=Attenuation_1, children [WavePlayer_3], properties {attenuationSettings=&/DanceOff_FarArea_Att}}
 {
  id=Attenuation_2
  children [Mixer_0]
  properties {attenuationSettings=&/DanceOff_FullAreaNoSpatial_Att}
 }
 {id=Mixer_0, children [WavePlayer_1,WavePlayer_2], properties {inputVolume [1,1]}}
 {id=Mixer_2, children [Attenuation_2, Attenuation_1], properties {inputVolume [1,1]}}
 {id=WavePlayer_1, properties {bLooping=true, soundWaveAssetPtr=/Figment_S07/Athena/Sounds/Quests/disco_puzzle_music_Layer2}}
 {id=WavePlayer_2, properties {bLooping=true, soundWaveAssetPtr=/Figment_S07/Athena/Sounds/Quests/disco_puzzle_music_Layer3}}
 {id=WavePlayer_3, properties {bLooping=true, soundWaveAssetPtr=/Figment_S07/Athena/Sounds/Quests/disco_puzzle_music_Layer1_reverb}}
]
properties {maxDistance=3400, soundClassObject=/Game/Sounds/FortSoundClassesAndModes/Fort_DanceMusic_Quest_PSM}
